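# Physics asset for the F_MED_Rover_Convoy female medium body cosmetic, apparently supporting secondary motion of its jacket bones.
# Capsule bodies cover both thighs, the right calf and a two-bone jacket chain per side; the left calf has none.
# Each jacket chain constrains its first bone to the same-side thigh and its second to the first, with swing limits and zero twist.
# The calf-to-thigh constraint leaves swing and twist free.
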PhysicsAsset /BRCosmetics/Characters/Player/Female/Medium/Bodies/F_MED_Rover_Convoy/Meshes/F_MED_Rover_Convoy_Physics
bodies {
 calf_r [{
   shapeType=Capsule
   length=35.5289
   radius=8.2421
   shapeName=calf_r_capsule
   center=<16.7466 2.3035 -0.3842>
   rotation=<rot 9.5325 -84.796 -88.3032>
  }]
 jacket1_l [{shapeType=Capsule, length=5, radius=5, shapeName=jacket1_r_capsule, center=<7.0244 -6 0>, rotation=<rot -90 14.0363 0.9675>}]
 jacket1_r [{shapeType=Capsule, length=5, radius=5, shapeName=jacket1_r_capsule, center=<-7.0244 6 0>, rotation=<rot 90 -14.0363 -34.0376>}]
 jacket2_l [{shapeType=Capsule, length=5, radius=5, shapeName=jacket1_r_capsule, center=<5.9478 -3 0>, rotation=<rot 90 90 69.9978>}]
 jacket2_r [{shapeType=Capsule, length=5, radius=5, shapeName=jacket1_r_capsule, center=<-5.9478 3 0>, rotation=<rot 90 90 69.9978>}]
 thigh_l [{shapeType=Capsule, length=30, radius=9.5, shapeName=thigh_l_capsule, center=<-19.2404 0 0>, rotation=<rot 90 -0.1583 -0.1583>}]
 thigh_r [{shapeType=Capsule, length=30, radius=9.5, shapeName=thigh_r_capsule, center=<19.2404 0 0>, rotation=<rot 90 -0.0099 -0.0099>}]
}
constraints [
 {bone1Name=jacket1_l, bone2Name=thigh_l, swing1LimitDegrees=45, swing2LimitDegrees=45, twistLimitDegrees=0}
 {bone1Name=jacket2_l, bone2Name=jacket1_l, swing1LimitDegrees=45, swing2LimitDegrees=45, twistLimitDegrees=0}
 {bone1Name=jacket1_r, bone2Name=thigh_r, swing1LimitDegrees=45, swing2LimitDegrees=45, twistLimitDegrees=0}
 {bone1Name=jacket2_r, bone2Name=jacket1_r, swing1LimitDegrees=45, swing2LimitDegrees=45, twistLimitDegrees=0}
 {
  bone1Name=calf_r
  bone2Name=thigh_r
  swing1LimitDegrees=45
  swing1Motion=Free
  swing2LimitDegrees=45
  swing2Motion=Free
  twistLimitDegrees=45
  twistMotion=Free
 }
]
