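# Physics asset for the F_MED_Rover_Convoy female character body, giving one boot-lace bone simple physics on the right calf.
# A small capsule on laces_l1 is constrained to a zero-length capsule on calf_twist_01_r.
# Swing and twist are all free, so the 45-degree limits do not restrict rotation.
# The lace bone and both shape names carry left-side naming despite the right-calf attachment.
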PhysicsAsset /BRCosmetics/Characters/Player/Female/Medium/Bodies/F_MED_Rover_Convoy/Meshes/F_MED_Rover_Convoy_PhysicsLacesRt1
bodies {
 calf_twist_01_r [{shapeType=Capsule, length=0, radius=2, shapeName=calf_twist_01_l_capsule}]
 laces_l1 [{shapeType=Capsule, length=1, radius=1, shapeName=laces_l_capsule, center=<-3 0 0>, rotation=<rot -90 0 0>}]
}
constraints [{
  bone1Name=laces_l1
  bone2Name=calf_twist_01_r
  swing1LimitDegrees=45
  swing1Motion=Free
  swing2LimitDegrees=45
  swing2Motion=Free
  twistLimitDegrees=45
  twistMotion=Free
 }]
