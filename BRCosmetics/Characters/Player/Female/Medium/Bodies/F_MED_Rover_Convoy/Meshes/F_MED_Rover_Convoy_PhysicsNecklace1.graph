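# Physics asset for the necklace on the F_MED_Rover_Convoy female medium player body cosmetic.
# Small capsule bodies on the necklace1 and necklace2 bones are each constrained directly to a larger spine_05 capsule that approximates the upper torso.
# Both constraints allow limited, equal swing on two axes and set the twist limit to zero, and the two necklace bodies have no constraint between each other.
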PhysicsAsset /BRCosmetics/Characters/Player/Female/Medium/Bodies/F_MED_Rover_Convoy/Meshes/F_MED_Rover_Convoy_PhysicsNecklace1
bodies {
 necklace1 [{shapeType=Capsule, length=2.25, radius=1.75, shapeName=necklace1_capsule, center=<5 0 1.5>, rotation=<rot -79.9997 -179.9998 -179.9999>}]
 necklace2 [{shapeType=Capsule, length=2.25, radius=1.75, shapeName=necklace2_capsule, center=<5 0 1.5>, rotation=<rot -79.9997 -179.9998 -179.9999>}]
 spine_05 [{
   shapeType=Capsule
   length=10.5543
   radius=9.7999
   shapeName=spine_05_capsule
   center=<-0.7683 -0.4085 0.0148>
   rotation=<rot -0.4013 -89.9999 0.1699>
  }]
}
constraints [
 {bone1Name=necklace2, bone2Name=spine_05, swing1LimitDegrees=30, swing2LimitDegrees=30, twistLimitDegrees=0}
 {bone1Name=necklace1, bone2Name=spine_05, swing1LimitDegrees=30, swing2LimitDegrees=30, twistLimitDegrees=0}
]
